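; Thrown concussion mine projectile that sticks where it lands, arms after a delay and shows team-colored beacon and radius effects.
; Once armed, valid-target overlaps trigger a warning and delayed detonation; damage or destruction of a non-building base detonates immediately.
; Detonation sends a server-only gameplay event with the owner as instigator, while a GameplayCue plays the explosion effects.
; A dying building base releases the mine, and it destroys itself if its instigator pawn is destroyed.
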(event EventBeginPlay
  (bind _returnvalue (Game|GetInstigator))
  ;[1 1 1] On Load
  (CallFunction|ListenforPawnValidity)
  (CallFunction|BindonDamage)
  (CallFunction|VFXSetTeamColors)
  (Variables|/Script/Engine.ProjectileMovementComponent|SetBounciness 0.2 (Variables|Default|GetProjectileMovement1))
  (if (Utilities|IsValid _returnvalue :!class /Script/Engine.KismetSystemLibrary)
    (CallFunction|SetTeam (Team|GetActorTeam _returnvalue))))

(event Custom|SetExplosion
  (params (Explosion%20Level :type Integer) (Activation_Radius :type Float%20%28double-precision%29) (PlayerPawn :type Fort%20Player%20Pawn%20Object%20Reference))
  (Variables|Default|SetActivation_Radius Activation_Radius)
  (Variables|Default|SetPlayer_Owner PlayerPawn)
  (Variables|Default|SetExplosionLevel Explosion%20Level)
  ;[0.1261 0.1093 1] Scaling the explosion
  (Variables|Default|SetEventExplosion (Variables|Default|GetExplosionRangeDefaultTag))
  (Variables|Default|SetExplosionRadiusGuide "500.000000,500.000000,3.000000")
  (Components|Sphere|SetSphereRadius (Variables|Default|GetOverlapCollider) (Variables|Default|GetOverlapColliderRadius) false))

(event Custom|SetProjectile
  (params (Init_Speed :type Float%20%28double-precision%29) (Gravity :type Float%20%28double-precision%29) (Activation_Delay :type Float%20%28double-precision%29))
  (bind _projectilemovement1 (Variables|Default|GetProjectileMovement1))
  ;[1 0.03636 0] Sets Prj movement from GA  (Not getting values!!!!)
  (Variables|/Script/Engine.ProjectileMovementComponent|SetInitialSpeed Init_Speed _projectilemovement1)
  (Variables|/Script/Engine.ProjectileMovementComponent|SetProjectileGravityScale Gravity _projectilemovement1)
  (Variables|Default|SetActivation_Delay Activation_Delay))

(event Collision|EventHit
  (params (MyComp :type Primitive%20Component%20Object%20Reference) (Other :type Actor%20Object%20Reference) (OtherComp :type Primitive%20Component%20Object%20Reference) (bSelfMoved :type Boolean) (HitLocation :type Vector :default "0, 0, 0") (HitNormal :type Vector :default "0, 0, 0") (NormalImpulse :type Vector :default "0, 0, 0") (Hit :type Hit%20Result%20Structure%20%28by%20ref%29))
  (bind _p_proximitymine_beacon_01 (Variables|Default|GetP_ProximityMine_Beacon_01))
  (bind _proxymine_mesh (Variables|Default|GetProxyMine_Mesh))
  (bind _player_owner (Variables|Default|GetPlayer_Owner))
  ;[1 1 1] On Projectile Hit - (Projectile being this actor)
  (Utilities|Casting|CastToPawn Other
    (:CastFailed
      (Variables|Default|SetImpact_Loc HitLocation)
      (Variables|Default|SetHitNormal HitNormal)
      (Game|Components|ProjectileMovement|StopSimulating (Variables|Default|GetProjectileMovement1) Hit)
      ;[1 1 1] Binds Event for when Building piece dies
      (bind _asbuilding_actor (Utilities|Casting|CastToBuildingActor Other)
        (:then
          (EventDispatchers|BindEventtoOnDied _asbuilding_actor (delegate Custom|BaseDied)))
        (:CastFailed
          (Game|BindEventtoOnDestroyed Other (delegate Custom|BaseDestroyed))))
      (Utilities|FlowControl|Sequence
        (:then_0
          ; MinePlaced
          (Utilities|FlowControl|Sequence
            (:then_0
              (Transformation|SetActorRotation (Math|Conversions|RotationFromXVector (Variables|Default|GetHitNormal)))
              (Transformation|SetRelativeRotation (Variables|Default|GetCollision_HalfSphere__1_Unit) "-90.000000,0.000000, 0" false false)
              (Transformation|SetRelativeRotation _proxymine_mesh "0, 0, 0" false false)
              (Transformation|AddRelativeRotation _proxymine_mesh (Math|Rotator|MakeRotator (Math|Random|RandomFloatinRange 0.0 1e+01)) false false)
              (Transformation|AddLocalOffset _proxymine_mesh "-14.000000, 0, 0" false false)
              (Collision|SetCollisionEnabled (Variables|Default|GetCollision_HalfSphere__1_Unit) "QueryAndPhysics")
              (Collision|SetCollisionEnabled (Variables|Default|GetRoot_Shpere_Collision) "NoCollision")
              (Audio|PlaySoundatLocation (Variables|Default|GetLand_Sound) (Variables|Default|GetImpact_Loc))
              (Components|Activation|Activate (Variables|Default|GetP_Land_Effect) false)
              (CallFunction|AttachFX)
              (Audio|Components|Audio|Play (Variables|Default|GetProximity_Mine_AmbientBeep_Cue) 0.0) ; Start looping sound
            )
            (:then_1
              (Utilities|FlowControl|Delay 0.1)
              (Components|Animation|PlayAnimation (Variables|Default|GetSK_ProximityMine) "/Game/Gadgets/Assets/ProximityMine/Mesh/SK_ProxMine_deployAnim.SK_ProxMine_deployAnim" false)
              (Utilities|FlowControl|Delay 0.5)
              (Components|DestroyComponent (Variables|Default|GetSK_ProximityMine))
              (Rendering|SetVisibility (Variables|Default|GetProxyMine_Mesh) true false))))
        (:then_1
          ; Activate Beacon
          (Transformation|SetWorldLocation (Variables|Default|GetP_Land_Effect) (Transformation|GetSocketLocation (Variables|Default|GetProxyMine_Mesh) "FX_Beacon") false false)
          (Transformation|SetWorldRotation (Variables|Default|GetP_ProximityMine_AoE_Reticle_01) (Math|Rotator|MakeRotfromZ (Variables|Default|GetHitNormal)) false false)
          (Variables|VFX|SetBeacon%20Can%20Play? true)
          (Utilities|FlowControl|Delay (Variables|VFX|GetBeacon%20Speedup%20Play%20Speed))
          (Variables|VFX|SetBeacon%20Speedup%20Play%20Speed (Math|Float|Clamp(Float) (- (Variables|VFX|GetBeacon%20Speedup%20Play%20Speed) (Variables|VFX|GetBeacon%20Speedup%20Play%20Speed%20Chip%20OffTime)) 0.09))
          (Components|Activation|SetActive (Variables|Default|GetP_ProximityMine_Beacon_01) true false)
          (CallFunction|BeaconLoop))
        (:then_2
          ; DistanceGuideEffect
          (Networking|CanExecuteCosmeticEvents
            (:True
              (Utilities|FlowControl|Delay 0.3)
              (Transformation|SetWorldScale3D (Variables|Default|GetExplodeDamageRingConstant_flat) (Variables|Default|GetExplosionRadiusGuide))
              (Transformation|SetWorldRotation (Variables|Default|GetExplodeDamageRingConstant_flat) (Math|Rotator|MakeRotfromZ (Variables|Default|GetHitNormal)) false false)
              (Development|SetHiddeninGame (Variables|Default|GetExplodeDamageRingConstant_flat) false false)
              (Transformation|SetWorldScale3D (Variables|Default|GetActivateRingConstant_Flat) (Math|Vector|MakeVector (Variables|Default|GetActivation_Radius) (Variables|Default|GetActivation_Radius) 3))
              (Transformation|SetWorldRotation (Variables|Default|GetActivateRingConstant_Flat) (Math|Rotator|MakeRotfromZ (Variables|Default|GetHitNormal)) false false)
              (Development|SetHiddeninGame (Variables|Default|GetActivateRingConstant_Flat) false false)
              (Transformation|SetWorldScale3D (Variables|Default|GetExplodeDamageRingConstant_Vertical) (Variables|Default|GetExplosionRadiusGuide))
              (Transformation|SetWorldRotation (Variables|Default|GetExplodeDamageRingConstant_Vertical) "0, 0, 0" false false)
              (Transformation|SetWorldScale3D (Variables|Default|GetActivateRingConstant_Vertical) (Math|Vector|MakeVector (Variables|Default|GetActivation_Radius) (Variables|Default|GetActivation_Radius) 3))
              (Transformation|SetWorldRotation (Variables|Default|GetActivateRingConstant_Vertical) "0, 0, 0" false false)
              (Development|SetHiddeninGame (Variables|Default|GetExplodeDamageRingConstant_Vertical) (< (Math|Trig|Acos(Degrees) (Math|Vector|DotProduct (Variables|Default|GetHitNormal) "0, 0,1.000000")) (Variables|VFX|GetAngleToShowAdditionalHUD)) false))))
        (:then_3
          ; Activate Disc
          (Networking|CanExecuteCosmeticEvents
            (:True
              (Transformation|SetWorldRotation (Variables|Default|GetSm_ProximityMine_RadarDisc_01) (Math|Rotator|MakeRotfromZ (Variables|Default|GetHitNormal)) false false)
              (Transformation|AddWorldOffset (Variables|Default|GetSm_ProximityMine_RadarDisc_01) "0, 0,2.000000" false false)
              (Development|SetHiddeninGame (Variables|Default|GetSm_ProximityMine_RadarDisc_01) false false)
              (bind (_direction _lerp) (Timeline|Scale%20Ring%20In PlayFromStart)
                (:Update
                  (Transformation|SetWorldScale3D (Variables|Default|GetSm_ProximityMine_RadarDisc_01) (Math|Vector|Lerp(Vector) "0.001000,0.001000,0.001000" (Math|Vector|MakeVector (Variables|Default|GetActivation_Radius) (Variables|Default|GetActivation_Radius) 20) _lerp))
                  (Transformation|SetWorldScale3D (Variables|Default|GetExplodeDamageRingConstant_flat) (Math|Vector|Lerp(Vector) "0.001000,0.001000,0.001000" (Variables|Default|GetExplosionRadiusGuide) _lerp)))))))
        (:then_4
          ;[1 1 1] Activation Delay
          (Utilities|FlowControl|Delay (Variables|Default|GetActivation_Delay))
          (Variables|Default|SetbActivated true)
          (Variables|VFX|SetBeacon%20Can%20Play? false)
          (Utilities|FlowControl|Sequence
            (:then_0
              (Networking|SwitchHasAuthority
                (:Authority
                  ;[1 0.6567 0] Handles case when guy is on mine when placed
                  (bind _array_element (Utilities|Array|ForEachLoopwithBreak (Collision|GetOverlappingActors (Variables|Default|GetPawnActors)))
                    (:LoopBody
                      (bind _returnvalue (CallFunction|IsValidTarget _array_element))
                      (if _returnvalue
                        (Utilities|FlowControl|Sequence
                          (:then_0
                            _)
                          (:then_1
                            _))
                        (Utilities|FlowControl|DoOnce false
                          (:Completed
                            (Utilities|FlowControl|Sequence
                              (:then_0
                                ; Activate Warning Beacon
                                (Effects|Components|ParticleSystem|SetVectorParameter _p_proximitymine_beacon_01 "CoreLightColor" "90.000000,4.400000,2.800000")
                                (Effects|Components|ParticleSystem|SetVectorParameter _p_proximitymine_beacon_01 "StreakColor" "90.000000,4.400000,2.800000")
                                (Rendering|Material|SetVectorParameterValue (Variables|VFX|GetMine_Mat) "Emissive Multiplier" "(R=90.000000,G=4.582951,B=2.792341,A=1.000000)" :!class /Script/Engine.MaterialInstanceDynamic)
                                (if (not (Variables|Default|GetIsDead?))
                                  (Components|Activation|Activate (Variables|Default|GetP_ProximityMine_Beacon_01) false)
                                  (Utilities|FlowControl|Delay 0.15)
                                  (Development|PrintString "beep before explode")))
                              (:then_1
                                ;[1 1 1] Delay before explode
                                (Utilities|FlowControl|Delay (Variables|Default|GetDelay%20Before%20Explode%20%28Overlap%29)) ; .25
                                (Utilities|FlowControl|DoOnce false
                                  (:Completed
                                    (Collision|SetCollisionEnabled (Variables|Default|GetProxyMine_Mesh) "NoCollision")
                                    ;[0.8855 0.04538 1]
                                    (Variables|Default|SetIsDead? true)
                                    (Networking|SwitchHasAuthority
                                      (:Authority
                                        (Utilities|IsValid _player_owner
                                          (:Is%20Valid
                                            (Ability|SendGameplayEventtoActor self (Variables|Default|GetEventExplosion) (Utilities|Struct|MakeGameplayEventData (Variables|Default|GetEventExplosion) _player_owner _player_owner self))
                                            (Development|PrintString "isdead" true true "(R=0.039889,G=1.000000,B=0.000000,A=1.000000)" 1e+01))))) ; If touched or damage\\ Only on server will it trigger explosion
                                  )))
                              (:then_2
                                (Audio|SpawnSoundAttached "/Game/Sounds/Fort_Gadgets/ProximityMine/fort_proximity_mine_activate_Cue.fort_proximity_mine_activate_Cue" (Variables|Default|GetProxyMine_Mesh))))))))))))
            (:then_1
              ; ScaleDownVFXandKill
              (Networking|CanExecuteCosmeticEvents
                (:True
                  (Components|Activation|Deactivate (Variables|Default|GetP_ProximityMine_Beacon_01))
                  (Rendering|Material|SetVectorParameterValue (Variables|VFX|GetMine_Mat) "Emissive Multiplier" (Utilities|GetTeamVFX (Variables|VFX|GetBeacon%20Mesh%20Color%20%28Friendly%29) (Variables|VFX|GetBeacon%20Mesh%20Color%20%28Enemy%29)) :!class /Script/Engine.MaterialInstanceDynamic)
                  (bind (_direction_2 _scaledown) (Timeline|ScaleDownAll PlayFromStart)
                    (:Update
                      (Transformation|SetWorldScale3D (Variables|Default|GetSm_ProximityMine_RadarDisc_01) (Math|Vector|Lerp(Vector) "0.001000,0.001000,0.001000" (Math|Vector|MakeVector (Variables|Default|GetActivation_Radius) (Variables|Default|GetActivation_Radius) 20) _scaledown))
                      (Transformation|SetWorldScale3D (Variables|Default|GetExplodeDamageRingConstant_flat) (Math|Vector|Lerp(Vector) "0.001000,0.001000,0.001000" (Variables|Default|GetExplosionRadiusGuide) _scaledown)))
                    (:Finished
                      (CallFunction|DestroyFX))))))))))))

(event Custom|BaseDestroyed
  (params (DestroyedActor :type Actor%20Object%20Reference))
  (bind _player_owner (Variables|Default|GetPlayer_Owner))
  (Utilities|FlowControl|DoOnce false
    (:Completed
      (Collision|SetCollisionEnabled (Variables|Default|GetProxyMine_Mesh) "NoCollision")
      ;[0.8855 0.04538 1]
      (Variables|Default|SetIsDead? true)
      (Networking|SwitchHasAuthority
        (:Authority
          (Utilities|IsValid _player_owner
            (:Is%20Valid
              (Ability|SendGameplayEventtoActor self (Variables|Default|GetEventExplosion) (Utilities|Struct|MakeGameplayEventData (Variables|Default|GetEventExplosion) _player_owner _player_owner self))
              (Development|PrintString "isdead" true true "(R=0.039889,G=1.000000,B=0.000000,A=1.000000)" 1e+01))))) ; If touched or damage\\ Only on server will it trigger explosion
    )))

(event Custom|BaseDied
  (params (DamagedActor :type Actor%20Object%20Reference) (Damage :type Float%20%28single-precision%29) (InstigatedBy :type Controller%20Object%20Reference) (DamageCauser :type Actor%20Object%20Reference) (HitLocation :type Vector) (FHitComponent :type Primitive%20Component%20Object%20Reference) (BoneName :type Name) (Momentum :type Vector))
  (Transformation|DetachActorFromActor(Deprecated) DamagedActor true)
  (Components|Movement|SetUpdatedComponent (Variables|Default|GetProjectileMovement1) (Variables|Default|GetRoot_Shpere_Collision))
  (Collision|SetCollisionEnabled (Variables|Default|GetRoot_Shpere_Collision) "QueryAndPhysics"))

(event Custom|BindOnDamage
  ;[1 1 1] On Damage
  (EventDispatchers|AssignOnDamaged (delegate Custom|OnDamaged_Event_0)))

(event Custom|OnDamaged_Event_0
  (params (DamagedActor :type Actor%20Object%20Reference) (Damage :type Float%20%28single-precision%29) (InstigatedBy :type Controller%20Object%20Reference) (DamageCauser :type Actor%20Object%20Reference) (HitLocation :type Vector) (FHitComponent :type Primitive%20Component%20Object%20Reference) (BoneName :type Name) (Momentum :type Vector))
  ;[1 1 1] On Damage
  (Networking|SwitchHasAuthority
    (:Authority
      (CallFunction|DamageExplode))))

(event OnComponentBeginOverlap(OverlapCollider)
  (params (OverlappedComponent :type Primitive%20Component%20Object%20Reference) (OtherActor :type Actor%20Object%20Reference) (OtherComp :type Primitive%20Component%20Object%20Reference) (OtherBodyIndex :type Integer) (bFromSweep :type Boolean) (SweepResult :type Hit%20Result%20Structure%20%28by%20ref%29))
  (bind _p_proximitymine_beacon_01 (Variables|Default|GetP_ProximityMine_Beacon_01))
  (bind _player_owner (Variables|Default|GetPlayer_Owner))
  ;[1 0.5923 0] Mine Triggered
  (bind _returnvalue (CallFunction|IsValidTarget OtherActor))
  (if _returnvalue
    (Utilities|FlowControl|DoOnce false
      (:Completed
        (Utilities|FlowControl|Sequence
          (:then_0
            ; Activate Warning Beacon
            (Effects|Components|ParticleSystem|SetVectorParameter _p_proximitymine_beacon_01 "CoreLightColor" "90.000000,4.400000,2.800000")
            (Effects|Components|ParticleSystem|SetVectorParameter _p_proximitymine_beacon_01 "StreakColor" "90.000000,4.400000,2.800000")
            (Rendering|Material|SetVectorParameterValue (Variables|VFX|GetMine_Mat) "Emissive Multiplier" "(R=90.000000,G=4.582951,B=2.792341,A=1.000000)" :!class /Script/Engine.MaterialInstanceDynamic)
            (if (not (Variables|Default|GetIsDead?))
              (Components|Activation|Activate (Variables|Default|GetP_ProximityMine_Beacon_01) false)
              (Utilities|FlowControl|Delay 0.15)
              (Development|PrintString "beep before explode")))
          (:then_1
            ;[1 1 1] Delay before explode
            (Utilities|FlowControl|Delay (Variables|Default|GetDelay%20Before%20Explode%20%28Overlap%29)) ; .25
            (Utilities|FlowControl|DoOnce false
              (:Completed
                (Collision|SetCollisionEnabled (Variables|Default|GetProxyMine_Mesh) "NoCollision")
                ;[0.8855 0.04538 1]
                (Variables|Default|SetIsDead? true)
                (Networking|SwitchHasAuthority
                  (:Authority
                    (Utilities|IsValid _player_owner
                      (:Is%20Valid
                        (Ability|SendGameplayEventtoActor self (Variables|Default|GetEventExplosion) (Utilities|Struct|MakeGameplayEventData (Variables|Default|GetEventExplosion) _player_owner _player_owner self))
                        (Development|PrintString "isdead" true true "(R=0.039889,G=1.000000,B=0.000000,A=1.000000)" 1e+01))))) ; If touched or damage\\ Only on server will it trigger explosion
              )))
          (:then_2
            (Audio|SpawnSoundAttached "/Game/Sounds/Fort_Gadgets/ProximityMine/fort_proximity_mine_activate_Cue.fort_proximity_mine_activate_Cue" (Variables|Default|GetProxyMine_Mesh))))))))

(event Custom|ListenForPawnValidity
  ;[1 1 1] FORT-59641: If owning pawn gets destroyed, this does as well
  (bind _asfort_pawn (Utilities|Casting|CastToFortPawn (Game|GetInstigator)))
  (Game|AssignOnDestroyed _asfort_pawn (delegate Custom|OnDestroyed_Event_0)))

(event Custom|DamageExplode
  (bind _player_owner (Variables|Default|GetPlayer_Owner))
  (Utilities|FlowControl|DoOnce false
    (:Completed
      (Collision|SetCollisionEnabled (Variables|Default|GetProxyMine_Mesh) "NoCollision")
      ;[0.8855 0.04538 1]
      (Variables|Default|SetIsDead? true)
      (Networking|SwitchHasAuthority
        (:Authority
          (Utilities|IsValid _player_owner
            (:Is%20Valid
              (Ability|SendGameplayEventtoActor self (Variables|Default|GetEventExplosion) (Utilities|Struct|MakeGameplayEventData (Variables|Default|GetEventExplosion) _player_owner _player_owner self))
              (Development|PrintString "isdead" true true "(R=0.039889,G=1.000000,B=0.000000,A=1.000000)" 1e+01))))) ; If touched or damage\\ Only on server will it trigger explosion
    )))

(event Custom|OnDestroyed_Event_0
  (params (DestroyedActor :type Actor%20Object%20Reference))
  ;[1 1 1] FORT-59641: If owning pawn gets destroyed, this does as well
  (Actor|DestroyActor))

(event GameplayCue|GameplayCue.Abilities.Activation.Generic.ProxyMine
  (params (EventType :type EGameplayCueEvent%20Enum :default "OnActive") (Parameters :type Gameplay%20Cue%20Parameters%20Structure%20%28by%20ref%29))
  (bind _p_explosioneffect (Variables|Default|GetP_ExplosionEffect))
  ;[1 0.4852 0] Gameplay cue triggers effects and death
  (Rendering|SetVisibility (Variables|Default|GetProxyMine_Mesh) false false)
  (Transformation|SetWorldRotation _p_explosioneffect "0, 0, 0" false false)
  (Components|Activation|Activate _p_explosioneffect false)
  (Components|DestroyComponent (Variables|Default|GetP_ProximityMine_AoE_Reticle_01))
  (Audio|SpawnSoundAttached "/Game/Sounds/Fort_GamePlay_Sounds/Explosives/ProximityMine_Explosion_Cue.ProximityMine_Explosion_Cue" _p_explosioneffect)
  (if (Utilities|Array|IsValidIndex (Variables|Default|GetRing%20Sizes) (Variables|Default|GetExplosionLevel))
    (Effects|Components|ParticleSystem|SetVectorParameter _p_explosioneffect "RingSize" (Utilities|Array|Get(acopy) (Variables|Default|GetRing%20Sizes) (Variables|Default|GetExplosionLevel)))
    (Effects|Components|ParticleSystem|SetVectorParameter _p_explosioneffect "SphereSize" (Utilities|Array|Get(acopy) (Variables|Default|GetSphere%20Sizes) (Variables|Default|GetExplosionLevel)))
    (else
      _))
  (Audio|Components|Audio|Stop (Variables|Default|GetProximity_Mine_AmbientBeep_Cue))
  (Camera|PlayWorldCameraShake "/Game/Weapons/FORT_RocketLaunchers/Blueprints/B_RocketLauncherCameraShake.B_RocketLauncherCameraShake_C" (Variables|Default|GetImpact_Loc) 5e+02 7.5e+02)
  (Utilities|SpawnDecalSimple "10.000000,30.000000,30.000000" "10.000000,50.000000,50.000000" (Variables|Default|GetImpact_Loc) "-90.000000,0,0" 1 1.2 "/Game/Weapons/Materials/MD_DecalSimple_Master_01.MD_DecalSimple_Master_01" "/Game/Effects/Fort_Effects/Gameplay/SmokeDustExplosions/Textures/T_GroundChar_01.T_GroundChar_01"))

(fn AttachFX
  (Networking|CanExecuteCosmeticEvents
    (:True
      (Transformation|AttachComponentToComponent (Variables|Default|GetP_ProximityMine_Beacon_01) (Variables|Default|GetProxyMine_Mesh) "FX_Beacon" "SnapToTarget" "SnapToTarget" "KeepWorld" true)
      (Transformation|AttachComponentToComponent (Variables|Default|GetP_ProximityMine_AoE_Reticle_01) (Variables|Default|GetProxyMine_Mesh) "FX_Root" "SnapToTarget" "SnapToTarget" "KeepWorld" true)
      (Transformation|AttachComponentToComponent (Variables|Default|GetSm_ProximityMine_RadarDisc_01) (Variables|Default|GetProxyMine_Mesh) "FX_Root" "SnapToTarget" "SnapToTarget" "KeepWorld" true)
      (Transformation|AttachComponentToComponent (Variables|Default|GetActivateRingConstant_Flat) (Variables|Default|GetProxyMine_Mesh) "FX_Root" "SnapToTarget" "SnapToTarget" "KeepWorld" true)
      (Transformation|AttachComponentToComponent (Variables|Default|GetExplodeDamageRingConstant_Vertical) (Variables|Default|GetProxyMine_Mesh) "FX_Root" "SnapToTarget" "SnapToTarget" "KeepWorld" true))))

(fn DestroyFX
  (Components|DestroyComponent (Variables|Default|GetSm_ProximityMine_RadarDisc_01)))

(fn IsValidTarget
  (params (Object :type Object%20Reference))
  (if (Variables|Default|GetbActivated)
    (bind _aspawn (Utilities|Casting|CastToPawn Object)
      (:then
        (switch Utilities|FlowControl|Switch|SwitchonEFortTeamAffiliation (Team|GetTeamAffiliation _aspawn self)
          (:Friendly
            (Variables|Default|SetIsValidEnemy false))
          (:Neutral
            (Variables|Default|SetIsValidEnemy true))
          (:Hostile
            _))
        (return (Variables|Default|GetIsValidEnemy)))
      (:CastFailed
        _))
    (else
      _)))

(fn OnRep_bActivated)

(fn OnRep_ExplosionRadiusGuide
  (bind _explosionradiusguide (Variables|Default|GetExplosionRadiusGuide))
  (bind _x (.x _explosionradiusguide))
  (bind _overlapcolliderradius (Variables|Default|GetOverlapColliderRadius))
  (bind _overlapcollider (Variables|Default|GetOverlapCollider))
  (if (== _x 5e+02)
    (Components|Sphere|SetSphereRadius _overlapcollider _overlapcolliderradius false)
    (elif (== _x 5e+02)
      (Components|Sphere|SetSphereRadius _overlapcollider _overlapcolliderradius false))))

(fn OnRep_IsDead?
  (Components|Activation|SetActive (Variables|Default|GetP_ProximityMine_Beacon_01) false false)
  (Variables|VFX|SetBeacon%20Can%20Play? false)
  (Development|SetHiddeninGame (Variables|Default|GetSm_ProximityMine_RadarDisc_01) true false))

(fn OnRep_IsPlaced)

(fn ConstructionScript
  (bind _returnvalue (Rendering|Material|CreateDynamicMaterialInstance (Variables|Default|GetActivateRingConstant_Flat) 0 "/Game/Gadgets/Assets/ProximityMine/Material/M_ProximityMine_Ring_01.M_ProximityMine_Ring_01" "None" :!class /Script/Engine.PrimitiveComponent))
  (Variables|VFX|SetActivate%20Ring%20MID _returnvalue)
  (bind _returnvalue_2 (Rendering|Material|CreateDynamicMaterialInstance (Variables|Default|GetProxyMine_Mesh) 0 "/Game/Gadgets/Assets/ProximityMine/Material/MIFORT_Gadgets_proximity_mine_Inst.MIFORT_Gadgets_proximity_mine_Inst" "None" :!class /Script/Engine.PrimitiveComponent))
  (Variables|VFX|SetMine_Mat _returnvalue_2)
  (bind _returnvalue_3 (Rendering|Material|CreateDynamicMaterialInstance (Variables|Default|GetSm_ProximityMine_RadarDisc_01) 0 "/Game/Gadgets/Assets/ProximityMine/Material/M_ProximityMine_Disc_01.M_ProximityMine_Disc_01" "None" :!class /Script/Engine.PrimitiveComponent))
  (Variables|VFX|SetRadar%20Disc%20MID _returnvalue_3)
  (bind _returnvalue_4 (Rendering|Material|CreateDynamicMaterialInstance (Variables|Default|GetActivateRingConstant_Vertical) 0 "/Game/Gadgets/Assets/ProximityMine/Material/M_ProximityMine_Ring_01.M_ProximityMine_Ring_01" "None" :!class /Script/Engine.PrimitiveComponent))
  (Variables|VFX|SetActivate%20Ring%20Vertical%20MID _returnvalue_4)
  (bind _returnvalue_5 (Rendering|Material|CreateDynamicMaterialInstance (Variables|Default|GetExplodeDamageRingConstant_flat) 0 "/Game/Gadgets/Assets/ProximityMine/Material/M_ProximityMine_Ring_01.M_ProximityMine_Ring_01" "None" :!class /Script/Engine.PrimitiveComponent))
  (Variables|VFX|SetDamageRingMID _returnvalue_5)
  (bind _returnvalue_6 (Rendering|Material|CreateDynamicMaterialInstance (Variables|Default|GetExplodeDamageRingConstant_Vertical) 0 0 "None" :!class /Script/Engine.PrimitiveComponent))
  (Variables|VFX|SetDamage%20Ring%20Vertical%20MID _returnvalue_6))

(fn VFX_SetTeamColors
  (bind (_color_out _friend_or_foe _friend) (Utilities|GetTeamVFX))
  (Effects|Components|ParticleSystem|SetVectorParameter (Variables|Default|GetP_ProximityMine_Beacon_01) "CoreLightColor" (Math|Conversions|ToVector(LinearColor) (Utilities|GetTeamVFX (Variables|VFX|GetBeacon%20Core%20Color%20%28Friendly%29) (Variables|VFX|GetBeacon%20Core%20Color%20%28Enemy%29))))
  (Effects|Components|ParticleSystem|SetVectorParameter (Variables|Default|GetP_ProximityMine_Beacon_01) "StreakColor" (Math|Conversions|ToVector(LinearColor) (Utilities|GetTeamVFX (Variables|VFX|GetBeacon%20Streak%20Color%20%28Friendly%29) (Variables|VFX|GetBeacon%20Streak%20Color%20%28Enemy%29))))
  (Rendering|Material|SetVectorParameterValue :self (Variables|VFX|GetMine_Mat) :ParameterName "Emissive Multiplier" :!class /Script/Engine.MaterialInstanceDynamic)
  (Rendering|Material|SetVectorParameterValue (Variables|VFX|GetRadar%20Disc%20MID) "Color Light" (Utilities|GetTeamVFX (Variables|VFX|GetRadar%20Disc%20Lights%20Color%20%28Friendly%29) (Variables|VFX|GetRadar%20Disc%20Lights%20Color%20%28Enemy%29)) :!class /Script/Engine.MaterialInstanceDynamic)
  (Rendering|Material|SetVectorParameterValue (Variables|VFX|GetRadar%20Disc%20MID) "Color Dark" (Utilities|GetTeamVFX (Variables|VFX|GetRadar%20Disc%20Dark%20Color%20%28Friendly%29) (Variables|VFX|GetRadar%20Disc%20Dark%20Color%20%28Enemy%29)) :!class /Script/Engine.MaterialInstanceDynamic)
  (Rendering|Material|SetVectorParameterValue (Variables|VFX|GetActivate%20Ring%20MID) "Color Inner" (Utilities|GetTeamVFX (Variables|VFX|GetActivate%20Ring%20Color%20Inner%20%28Friendly%29) (Variables|VFX|GetActivate%20Ring%20Color%20Inner%20%28Enemy%29)) :!class /Script/Engine.MaterialInstanceDynamic)
  (Rendering|Material|SetVectorParameterValue (Variables|VFX|GetActivate%20Ring%20MID) "Color Outer" (Utilities|GetTeamVFX (Variables|VFX|GetActivate%20Ring%20Color%20Outer%20%28Friendly%29) (Variables|VFX|GetActivate%20Ring%20Color%20Outer%20%28Enemy%29)) :!class /Script/Engine.MaterialInstanceDynamic)
  (Effects|Components|ParticleSystem|SetFloatParameter (Variables|Default|GetP_ProximityMine_AoE_Reticle_01) "TeamColor" (Math|Conversions|ToFloat(Boolean) _friend)))
